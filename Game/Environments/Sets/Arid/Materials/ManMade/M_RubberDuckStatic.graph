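# Opaque Default Lit material for a static rubber duck prop in the Arid environment set, flagged for Nanite and instanced static meshes.
# Diffuse and Normal texture parameters supply base color and normal, while the red channel of a fixed mask texture drives both Specular and Roughness.
# The combined material attributes pass through MF_CreativeMoveTool, which by its name adds Creative move-tool support.
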Material /Game/Environments/Sets/Arid/Materials/ManMade/M_RubberDuckStatic
graph {
 properties {
  bUseMaterialAttributes=true
  bUsedWithInstancedStaticMeshes=true
  bUsedWithNanite=true
  pixelDepthOffsetMode=PDOM_Legacy
  usedShadingModels=Default Lit
 }
 nodes [
  n1 = TextureSample() {samplerType=SAMPLERTYPE_Masks, texture=/Game/Environments/Sets/Arid/Textures/Rubber_Duck_S}
  MaterialFunctionCall(MA In=MakeMaterialAttributes!(
    BaseColor=TextureSampleParameter2D() {parameterName=Diffuse, texture=/Game/Environments/Sets/Arid/Textures/Rubber_Duck_D}.RGB
    Metallic
    Specular=@n1.R
    Roughness=@n1.R
    Anisotropy
    EmissiveColor
    Opacity
    OpacityMask
    Normal=TextureSampleParameter2D() {parameterName=Normal, samplerType=SAMPLERTYPE_Normal, texture=/Game/Environments/Sets/Arid/Textures/Rubber_Duck_N}.RGB
    Tangent
    WorldPositionOffset
    SubsurfaceColor
    ClearCoat
    ClearCoatRoughness
    AmbientOcclusion
    Refraction
    CustomizedUVs_0
    CustomizedUVs_1
    CustomizedUVs_2
    CustomizedUVs_3
    CustomizedUVs_4
    CustomizedUVs_5
    CustomizedUVs_6
    CustomizedUVs_7
    PixelDepthOffset
    ShadingModel
    Displacement
   )) {materialFunction=&MF_CreativeMoveTool}
 ]
}
